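; PCG subgraph that computes an ambient occlusion value for mesh terrain points and writes it to a named attribute, "Occlusion" by default.
; It keeps only point data, adds a temporary __AO attribute, then runs a Custom HLSL GPU step driven by radius, height scale, bias, strength and power parameters.
; The result is copied to the output attribute, and all double-underscore temporary attributes are deleted before output.
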(graph Write_MeshTerrain_AO_CoreProcess
  (params (OcclusionRadius :type Double :default 5e+04) (OcclusionHeightScale :type Double :default 1.0) (OcclusionBias :type Double) (OcclusionStrength :type Double :default 1.0) (OcclusionPower :type Double :default 1.0) (OutputAttribute :type String :default "Occlusion"))
  (bind FilterDataByType_0 (Filter%20Data%20By%20Type :TargetType (struct :Ids [(struct :Struct /Script/PCG.PCGDataTypeInfoPoint)])))
  (bind AddAttribute_0 (Add%20Attribute :OutputTarget __AO))
  (bind CustomHLSL_0 (Custom%20HLSL :InputPins [(struct :Label In :AllowedTypes (struct :Ids [(struct :Struct /Script/PCG.PCGDataTypeInfoPoint)]) :PinStatus Required :bAllowMultipleConnections false) (struct :Label OcclusionRadius :AllowedTypes (struct :Ids [(struct :Struct /Script/PCG.PCGDataTypeInfoParam)]) :bAllowMultipleConnections false) (struct :Label OcclusionHeightScale :AllowedTypes (struct :Ids [(struct :Struct /Script/PCG.PCGDataTypeInfoParam)]) :bAllowMultipleConnections false) (struct :Label OcclusionBias :AllowedTypes (struct :Ids [(struct :Struct /Script/PCG.PCGDataTypeInfoParam)]) :bAllowMultipleConnections false) (struct :Label OcclusionStrength :AllowedTypes (struct :Ids [(struct :Struct /Script/PCG.PCGDataTypeInfoParam)]) :bAllowMultipleConnections false) (struct :Label OcclusionPower :AllowedTypes (struct :Ids [(struct :Struct /Script/PCG.PCGDataTypeInfoParam)]) :bAllowMultipleConnections false)] :OutputPins [(struct :PropertiesGPU (struct :PinsToInititalizeFrom [In]) :Label Out :AllowedTypes (struct :Ids [(struct :Struct /Script/PCG.PCGDataTypeInfoPoint)]))] :OcclusionBias OcclusionBias :OcclusionHeightScale OcclusionHeightScale :OcclusionPower OcclusionPower :OcclusionRadius OcclusionRadius :OcclusionStrength OcclusionStrength)) ; Erosion effect
  (bind CopyAttributes_0 (Copy%20Attributes :InputSource __AO :OutputTarget-param OutputAttribute)) ; Copy temporary attribute\n to OutputAttribute
  (bind DeleteAttributes_1 (Delete%20Attributes :Operator Matches :SelectedAttributes "__*")) ; Delete temporary attribute
  (connect FilterDataByType_0 InsideFilter AddAttribute_0 In)
  (connect CustomHLSL_0 Out CopyAttributes_0 Source)
  (connect CustomHLSL_0 Out CopyAttributes_0 Target)
  (connect AddAttribute_0 Out CustomHLSL_0 In)
  (connect CopyAttributes_0 Out DeleteAttributes_1 In)
  (connect DefaultInputNode In FilterDataByType_0 In)
  (output Out DeleteAttributes_1))
