(graph Write_MeshTerrain_CoreProcess
  (params (MeshPartitionBoundSize :type Vector :default (1e+05 1e+05 1e+05)) (AffectedMeshPartition :type (SoftObject /Script/MeshPartition.MeshPartition)) (ModifierPriorityLayer :type Name) (ModifierSubPriority :type Double) (Channels :type (Array Name)))
  (bind FilterDataByType_0 (Filter%20Data%20By%20Type :TargetType (struct :Ids [(struct :Struct /Script/PCG.PCGDataTypeInfoPoint)])))
  (bind MegaMeshWrite_1 (MeshPartition%20Write :AffectedMegaMesh-param AffectedMeshPartition :Channels-param Channels :Priority-param ModifierSubPriority :Type-param ModifierPriorityLayer))
  (bind MegaMeshQuery_1 (Mesh%20Partition%20Query :QueryParams (struct :QueryType Intermediate :SubPriority 1.0 :bInclusive true :bRecomputeVertexNormals true) :Channels Channels :LayerName ModifierPriorityLayer :MegaMeshOverride AffectedMeshPartition :SubPriority ModifierSubPriority))
  ;[0 0 0] Replacing PCG volume bound size
  (bind AttributeMathsOp_0 (Attribute%20Maths%20Op :Operation Divide :InA MeshPartitionBoundSize))
  (bind AttributeMathsOp_3 (Attribute%20Maths%20Op :Operation Multiply))
  (bind CreatePoints_2 (Create%20Points))
  (connect AttributeMathsOp_0 Out AttributeMathsOp_3 InA)
  (connect AttributeMathsOp_0 Out CreatePoints_2 BoundsMax)
  (connect AttributeMathsOp_3 Out CreatePoints_2 BoundsMin)
  (connect DefaultInputNode In FilterDataByType_0 In)
  (connect CreatePoints_2 Out MegaMeshQuery_1 Bounding%20Shape)
  (connect MegaMeshWrite_1 Execution%20Dependency MegaMeshQuery_1 Execution%20Dependency)
  (connect CreatePoints_2 Out MegaMeshWrite_1 Bounding%20Shape)
  (connect FilterDataByType_0 InsideFilter MegaMeshWrite_1 In)
  (output Out MegaMeshQuery_1))
